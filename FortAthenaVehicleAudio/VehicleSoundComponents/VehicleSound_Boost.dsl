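; Vehicle sound component that handles boost audio: start, end, ready and not-available sounds.
; On init and whenever a pawn enters seat 0, it pushes sounds, fade times and the ready threshold into a native boost context.
; The not-available sound is included only when the driver is locally viewed.
; OnBoost and OnBoostFinished spawn attached start and end sounds, each fading out the other's stored audio component.
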(event VehicleAudio|EventOnInit
  (|CallParentFunction)
  (CallFunction|SetNativeData)
  (Vehicle|BindEventtoOnPawnEnteredVehicleatSeat (VehicleSound|GetVehicleActor) (EventDispatchers|CreateEvent :!event OnPawnEnteredVehicle)))

(fn OnBoost
  (bind _boostsound (Variables|Boost|GetBoostSound)
    (:then
      (bind _returnvalue (Audio|SpawnSoundAttached _boostsound (VehicleSound|GetRootComponent)))
      (Variables|Default|SetBoostStartAudio _returnvalue)
      (bind _boostendaudio (Variables|Default|GetBoostEndAudio)
        (:then
          (Audio|Components|Audio|FadeOut _boostendaudio (Variables|Boost|GetBoostEndFadeOut) 0.0 "Linear" :!class /Script/Engine.AudioComponent))
        (:else)))
    (:else)))

(fn OnBoostFinished
  (bind _returnvalue (Audio|SpawnSoundAttached (Variables|Boost|GetBoostEndSound) (VehicleSound|GetRootComponent)))
  (Variables|Default|SetBoostEndAudio _returnvalue)
  (bind _booststartaudio (Variables|Default|GetBoostStartAudio)
    (:then
      (Audio|Components|Audio|FadeOut _booststartaudio (Variables|Boost|GetBoostFadeOut) 0.0 "Linear" :!class /Script/Engine.AudioComponent))
    (:else)))

(fn OnPawnEnteredVehicle
  (params (Vehicle :type Fort%20Vehicle%20Interface%20Interface%20%28by%20ref%29) (PlayerPawn :type Fort%20Pawn%20Object%20Reference) (SeatIndex :type Integer))
  (if (== SeatIndex 0)
    (Variables|Default|SetDriverIsLocal (Pawn|IsLocallyViewed PlayerPawn))
    (CallFunction|SetNativeData)
    (return)))

(fn SetNativeData
  (Variables|SetContext_BoostStartSound (Utilities|Struct|MakeFortVehicleSoundComponentBoostContext (Variables|Boost|GetBoostSound) (Variables|Boost|GetBoostEndSound) (Variables|Boost|GetBoost%20Ready%20Sound) (select (Variables|Default|GetDriverIsLocal) (Variables|Boost|GetBoost%20Not%20Available%20Sound) 0) (Variables|Boost|GetBoostFadeOut) (Variables|Boost|GetBoostEndFadeOut) (Variables|Boost|GetBoost%20Ready%20Threshold))))
