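; Vehicle sound component that plays damage-related audio for a Fortnite vehicle: damage states, pre-destroy, explosion, tire pops, window breaks and scraping.
; Health changes step a damage index through mapped thresholds, crossfading attached sounds, with water variants chosen on entering or leaving water.
; Damageable-part events match part names to place tire and window sounds at sockets.
; Scraping hands a MetaSound to native context and has an interpolated, speed-driven parameter update.
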(event VehicleAudio|EventOnInit
  (bind _returnvalue (VehicleSound|GetVehicleActor))
  (|CallParentFunction)
  (CallFunction|SetNativeData)
  (Utilities|IsValid _returnvalue
    (:Is%20Valid
      (Health|BindEventtoOnHealthChanged _returnvalue (EventDispatchers|CreateEvent :!event On%20Health%20Change))
      (Collision|AssignOnDamageablePartHealthChanged _returnvalue (EventDispatchers|CreateEvent :!event Parts%20Damaged))
      (Vehicle|BindEventtoOnVehicleEnteredWaterBody _returnvalue (EventDispatchers|CreateEvent :!event Enter%20Water))
      (Vehicle|AssignOnVehicleExitedWaterBody _returnvalue (EventDispatchers|CreateEvent :!event Exit%20Water))
      (Health|UpdateCurrentHealth)
      (Variables|Private|SetDamageIndex (Utilities|Map|Length (Variables|DamageStates|GetDamageStatValues))))))

(event VehicleAudio|EventOnShutdown
  (bind _returnvalue (VehicleSound|GetVehicleActor))
  (|CallParentFunction)
  (EndSound|StopDamageAudio)
  (Utilities|IsValid _returnvalue
    (:Is%20Valid
      (Health|UnbindEventfromOnHealthChanged _returnvalue (EventDispatchers|CreateEvent :!event On%20Health%20Change))
      (Collision|UnbindEventfromOnDamageablePartHealthChanged _returnvalue (EventDispatchers|CreateEvent :!event Parts%20Damaged))
      (Vehicle|UnbindEventfromOnVehicleEnteredWaterBody _returnvalue (EventDispatchers|CreateEvent :!event Enter%20Water))
      (Vehicle|UnbindEventfromOnVehicleExitedWaterBody _returnvalue (EventDispatchers|CreateEvent :!event Exit%20Water)))))

(event VehicleSoundDamageState|Native|EventOnExplode
  (params (LastDamageInstigator :type Controller%20Object%20Reference) (Vehicle :type Fort%20Athena%20Vehicle%20Object%20Reference))
  (EndSound|OnExplode LastDamageInstigator Vehicle))

(event VehicleSoundDamageState|Native|EventOnScrapingSidesOrTop
  (params (bIsScraping :type Boolean))
  (Scraping|OnScrapingSidesOrTop bIsScraping))

(event VehicleSoundDamageState|Native|EventOnSleepStateChanged
  (params (bIsAwake :type Boolean))
  (Default|OnSleepStateChanged bIsAwake))

(fn EnterWater
  (params (Vehicle :type Fort%20Athena%20Vehicle%20Object%20Reference) (WaterBody :type Fort%20Water%20Body%20Actor%20Object%20Reference))
  (Variables|Private|SetIsInWater true)
  (if (Variables|Private|GetIsDamaged)
    (Damage|PlayDamageStateSound)))

(fn ExitWater
  (params (Vehicle :type Fort%20Athena%20Vehicle%20Object%20Reference) (WaterBody :type Fort%20Water%20Body%20Actor%20Object%20Reference))
  (Variables|Private|SetIsInWater false)
  (if (Variables|Private|GetIsDamaged)
    (Damage|PlayDamageStateSound)))

(fn GetNormalizedHealth
  (bind _returnvalue (VehicleSound|GetVehicleActor))
  (Utilities|IsValid _returnvalue
    (:Is%20Valid
      (return (Math|Float|SafeDivide (Class|FortVehicleInterface|GetHealth _returnvalue) (Class|FortVehicleInterface|GetMaxHealth _returnvalue))))
    (:Is%20Not%20Valid
      (return 1.0))))

(fn OnEndPlay
  (params (Actor :type Actor%20Object%20Reference) (EndPlayReason :type EEndPlayReason%20Enum :default "Destroyed"))
  (EndSound|StopDamageAudio)
  (Scraping|OnScrapingSidesOrTop false))

(fn OnExplode
  (params (LastDamageInstigator :type Controller%20Object%20Reference) (Vehicle :type Fort%20Athena%20Vehicle%20Object%20Reference))
  (bind _isinwater (Variables|Private|GetIsInWater))
  (bind _waterexplodesound (Variables|DamageStates|GetWaterExplodeSound))
  (bind _explodesound (Variables|DamageStates|GetExplodeSound))
  (bind _returnvalue (select _isinwater _waterexplodesound _explodesound))
  (Variables|Private|SetIsDestroyed true)
  (Health|UnbindEventfromOnHealthChanged (VehicleSound|GetVehicleActor) (EventDispatchers|CreateEvent :!event On%20Health%20Change))
  (Utilities|IsValid _returnvalue
    (:Is%20Valid
      (bind _returnvalue_2 (Audio|SpawnSoundAttached _returnvalue (Variables|/Script/Engine.Actor|GetRootComponent (Components|GetOwner))))
      (Variables|Default|SetExplodeAudio _returnvalue_2))
    (:Is%20Not%20Valid
      _))
  (bind _predestroyaudiocomponent (Variables|Default|GetPreDestroyAudioComponent)
    (:then
      (Audio|Components|Audio|FadeOut _predestroyaudiocomponent 0.5 0.0 "Linear" :!class /Script/Engine.AudioComponent))
    (:else)))

(fn OnHealthChange
  (LocalVariables|SetNormalizedHealth (Health|GetNormalizedHealth)) ; Calculate once to avoid calling NormalizeHealth many times within this function.
  (Variables|Private|SetIsInWater (Class|FortVehicleInterface|IsinWater (VehicleSound|GetVehicleActor)))
  ;[1 1 1] Check if we are in damage value range, otherwise kill audio.
  (if (Math|Boolean|ORBoolean (and (>= (LocalVariables|GetNormalizedHealth) 1.0) (< (LocalVariables|GetNormalizedHealth) 0)) (> (LocalVariables|GetNormalizedHealth) (Utilities|Map|Find (Variables|DamageStates|GetDamageStatValues) (Utilities|Map|Length (Variables|DamageStates|GetDamageStatValues)))) (Variables|Private|GetIsDestroyed))
    ;[1 1 1] Should we subtract or add index based on current health
    (EndSound|StopDamageAudio)
    (elif (> (LocalVariables|GetNormalizedHealth) (Variables|Private|GetCurrentHealth))
      ;[1 1 1] Subtract damage index
      (if (== (Variables|Private|GetDamageIndex) 0)
        (Variables|Private|SetDamageIndex 1)
        (else
          _))
      (if (>= (LocalVariables|GetNormalizedHealth) (Utilities|Map|Find (Variables|DamageStates|GetDamageStatValues) (Variables|Private|GetDamageIndex)))
        (Variables|Private|SetDamageIndex (+ (Math|Integer|Min(Integer) (Variables|Private|GetDamageIndex) (Utilities|Map|Length (Variables|DamageStates|GetDamageStatValues))) 1))
        (Damage|PlayDamageStateSound)
        (else
          _))
      (Health|UpdateCurrentHealth)
      (else
        ;[1 1 1] Add damage index
        (if (== (Variables|Private|GetDamageIndex) 0)
          _
          (elif (<= (LocalVariables|GetNormalizedHealth) (Utilities|Map|Find (Variables|DamageStates|GetDamageStatValues) (Variables|Private|GetDamageIndex)))
            (Damage|SetDamageStateSounds)
            (Damage|PlayDamageStateSound)
            (Variables|Private|SetDamageIndex (- (Variables|Private|GetDamageIndex) 1))
            (else
              _)))
        (Health|UpdateCurrentHealth)
        (if (<= (LocalVariables|GetNormalizedHealth) 0)
          (Damage|OnPreDestroy))))))

(fn OnPreDestroy
  (bind _isinwater (Variables|Private|GetIsInWater))
  (bind _prewaterdestroysoundcue (Variables|DamageStates|GetPreWaterDestroySoundCue))
  (bind _predestroysoundcue (Variables|DamageStates|GetPreDestroySoundCue))
  (bind _returnvalue (select _isinwater _prewaterdestroysoundcue _predestroysoundcue))
  (EndSound|StopDamageAudio)
  (Utilities|IsValid _returnvalue
    (:Is%20Valid
      (bind _returnvalue_2 (Audio|SpawnSoundAttached _returnvalue (Variables|/Script/Engine.Actor|GetRootComponent (Components|GetOwner))))
      (Variables|Default|SetPreDestroyAudioComponent _returnvalue_2))))

(fn OnScrapingSidesorTop
  (params (bIsScraping :type Boolean))
  (Variables|Private|SetIsScraping bIsScraping))

(fn OnSleepStateChanged
  (params (bIsAwake :type Boolean))
  (if bIsAwake
    _))

(fn PartsDamaged
  (params (PartConfig :type Vehicle%20Damageable%20Part%20Config%20Structure%20%28by%20ref%29) (Part :type Vehicle%20Damageable%20Part%20Structure%20%28by%20ref%29) (bNewlyDisabled :type Boolean) (bNewlyEnabled :type Boolean))
  (bind (_shapename _bonename) (Utilities|Struct|BreakVehicleDamageablePartConfig PartConfig))
  (Tires|TirePopped _shapename _bonename bNewlyDisabled)
  (if bNewlyDisabled
    (Window|WindowBreak _shapename _bonename)))

(fn PlayDamageStateSound
  (if (Variables|Private|GetIsInWater)
    (bind _currentdamagewatersound (Variables|Private|GetCurrentDamageWaterSound)
      (:then
        (LocalVariables|SetDamageSound _currentdamagewatersound))
      (:else
        (bind _currentdamagesound (Variables|Private|GetCurrentDamageSound)
          (:then
            (LocalVariables|SetDamageSound _currentdamagesound))
          (:else
            (bind _damageaudio (Variables|Default|GetDamageAudio)
              (:then
                (Audio|Components|Audio|FadeOut _damageaudio 0.75 0.0 "Linear" :!class /Script/Engine.AudioComponent)
                (LocalVariables|SetDamageSound 0))
              (:else))))))
    (bind _damageaudio_2 (Variables|Default|GetDamageAudio)
      (:then
        (if (!= (Variables|/Script/Engine.AudioComponent|GetSound _damageaudio_2) (LocalVariables|GetDamageSound))
          (Audio|Components|Audio|FadeOut (Variables|Default|GetDamageAudio) 0.5 0.0 "Linear" :!class /Script/Engine.AudioComponent)))
      (:else
        _))
    (bind _returnvalue (Audio|SpawnSoundAttached (LocalVariables|GetDamageSound) (Variables|/Script/Engine.Actor|GetRootComponent (Components|GetOwner)) "None" "0.000000,0.000000,-50.000000"))
    (bind _output_get (Variables|Default|SetDamageAudio _returnvalue))
    (Audio|Parameter|SetBooleanParameter _output_get "IsDamaged" (Variables|Private|GetIsDamaged) :!class /Script/AudioExtensions.AudioParameterControllerInterface)
    (Variables|Private|SetIsDamaged true)
    (else
      _)))

(fn SetDamageStateSounds
  (bind (_value _returnvalue) (Utilities|Map|Find (Variables|DamageStates|GetWaterDamageStateSounds) (Variables|Private|GetDamageIndex)))
  (Variables|Private|SetCurrentDamageWaterSound (select _returnvalue _value (Utilities|Map|Find (Variables|DamageStates|GetWaterDamageStateSounds) 1)))
  (Variables|Private|SetCurrentDamageSound (Utilities|Map|Find (Variables|DamageStates|GetDamageStateAudio) (Variables|Private|GetDamageIndex))))

(fn SetNativeData
  (Utilities|Struct|SetmembersinFortVehicleSoundComponentDamageStateContext (Variables|VehicleSoundDamageState|Native|GetContext) "/VehicleAudio/MetaSounds/MSS_Vehicle_Scraping.MSS_Vehicle_Scraping" "Scraping" "Speed"))

(fn StopDamageAudio
  (Variables|Private|SetDamageIndex (Utilities|Map|Length (Variables|DamageStates|GetDamageStatValues)))
  (Variables|Private|SetCurrentHealth (Health|GetNormalizedHealth))
  (bind _damageaudio (Variables|Default|GetDamageAudio)
    (:then
      (Audio|Components|Audio|FadeOut _damageaudio 0.75 0.0 "Linear" :!class /Script/Engine.AudioComponent)
      (Variables|Private|SetIsDamaged false))
    (:else
      _))
  (bind _predestroyaudiocomponent (Variables|Default|GetPreDestroyAudioComponent)
    (:then
      (Audio|Components|Audio|FadeOut _predestroyaudiocomponent 0.5 0.0 "Linear" :!class /Script/Engine.AudioComponent))
    (:else)))

(fn StopScrapeAudio
  (bind _scrapingaudio (Variables|Default|GetScrapingAudio)
    (:then
      (Audio|Components|Audio|FadeOut _scrapingaudio 1.0 0.0 "Linear" :!class /Script/Engine.AudioComponent))
    (:else)))

(fn TirePopped
  (params (InName :type Name) (BoneName :type Name) (NewlyDisabled :type Boolean))
  (if (Utilities|String|Contains (Utilities|String|ToString(Name) InName) (Utilities|String|ToString(Name) "Wheel"))
    (Utilities|FlowControl|Sequence
      (:then_0
        (switch name InName
          (:Default
            (LocalVariables|SetLocalTireSotcket "None"))
          (:wheel_fr
            (LocalVariables|SetLocalTireSotcket "tire_fr_r_socket"))
          (:wheel_fl
            (LocalVariables|SetLocalTireSotcket "tire_fr_l_socket"))
          (:wheel_br
            (LocalVariables|SetLocalTireSotcket "tire_bk_r_socket"))
          (:wheel_bl
            (LocalVariables|SetLocalTireSotcket "tire_bk_l_socket"))
          (:wheel_bl_f
            (LocalVariables|SetLocalTireSotcket "tire_mid_l_socket"))
          (:wheel_br_f
            (LocalVariables|SetLocalTireSotcket "tire_mid_r_socket"))))
      (:then_1
        (Audio|SpawnSoundAttached :Sound (select (Variables|GetNewlyDisabled) (Variables|Tire|GetTirePopSound) "/VehicleAudio/Cues/Veh_Valet_Tire_Impact_Cue.Veh_Valet_Tire_Impact_Cue") :AttachToComponent (Variables|/Script/FortniteGame.FortAthenaVehicle|GetSkeletalMesh (VehicleSound|GetVehicleActor)) :AttachPointName "None" :Location (Transformation|GetSocketLocation (Variables|/Script/FortniteGame.FortAthenaVehicle|GetSkeletalMesh (VehicleSound|GetVehicleActor)) (LocalVariables|GetLocalTireSotcket)) :LocationType "KeepWorldPosition" :bStopWhenAttachedToDestroyed false :VolumeMultiplier 1.0 :PitchMultiplier 1.0 :StartTime 0.0 :bAutoDestroy true)))))

(fn UpdateCurrentHealth
  (Variables|Private|SetCurrentHealth (Health|GetNormalizedHealth)))

(fn UpdateScrapingAudio
  (bind _handlescraping (Variables|Private|GetHandleScraping))
  (bind _output_get (Variables|Private|SetScrapingParam (Math|Interpolation|FInterpTo (Variables|Private|GetScrapingParam) (Math|Conversions|ToFloat(Boolean) (Math|Boolean|ANDBoolean (not (Vehicle|IsAsleep (VehicleSound|GetVehicleActor))) (Variables|Private|GetIsScraping) (> (Math|Float|MapRangeClamped (Vehicle|GetTotalSpeedKmh (VehicleSound|GetVehicleActor)) 0.0 (Class|FortVehicleInterface|GetHUDTopSpeed (VehicleSound|GetVehicleActor)) 0.2 1.25) 0.1) (not (Variables|Private|GetIsDestroyed)))) (Utilities|Time|GetWorldDeltaSeconds) (Math|Float|SelectFloat 5.0 1e+01 (Math|Boolean|ANDBoolean (not (Vehicle|IsAsleep (VehicleSound|GetVehicleActor))) (Variables|Private|GetIsScraping) (> (Math|Float|MapRangeClamped (Vehicle|GetTotalSpeedKmh (VehicleSound|GetVehicleActor)) 0.0 (Class|FortVehicleInterface|GetHUDTopSpeed (VehicleSound|GetVehicleActor)) 0.2 1.25) 0.1) (not (Variables|Private|GetIsDestroyed)))))))
  (Audio|Parameter|SetFloatParameter (Variables|Default|GetScrapingAudio) "Scraping" _output_get :!class /Script/AudioExtensions.AudioParameterControllerInterface)
  (Audio|Parameter|SetFloatParameter (Variables|Default|GetScrapingAudio) "Speed" (Math|Float|MapRangeClamped (Vehicle|GetTotalSpeedKmh (VehicleSound|GetVehicleActor)) 0.0 (Class|FortVehicleInterface|GetHUDTopSpeed (VehicleSound|GetVehicleActor)) 0.2 1.25) :!class /Script/AudioExtensions.AudioParameterControllerInterface)
  (if (<= (Variables|Private|GetScrapingParam) 0)
    (Scraping|StopScrapeAudio)
    (if (Utilities|Time|IsValidTimerHandle _handlescraping)
      (Utilities|Time|ClearandInvalidateTimerbyHandle _handlescraping)
      (Variables|Private|SetScrapingParam 0.0))))

(fn WindowBreak
  (params (ShapeName :type Name) (BroneName :type Name))
  (if (Utilities|String|Contains (Utilities|String|ToString(Name) ShapeName) (Utilities|String|ToString(Name) (Variables|Window|GetWindowName)))
    (Audio|PlaySoundatLocation (Variables|Window|GetWindowBreakSoundCue) (Transformation|GetSocketLocation (Class|FortVehicleInterface|GetSkeletalMeshComponent (VehicleSound|GetVehicleActor)) BroneName) "0, 0, 0" 1.0 1.0 0.0 0 0 (VehicleSound|GetVehicleActor))))
