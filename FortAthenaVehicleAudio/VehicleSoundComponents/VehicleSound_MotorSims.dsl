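; Vehicle sound component that configures the Audio Motor Sim components driving a vehicle's simulated engine audio.
; On audio init it calls the parent, then runs six setters: throttle state, resistance, reverse, rev limiter, motor physics and velocity sync.
; Each finds its sim component by class on the audio controller and, if valid, copies this component's tuning variables onto it, from gear ratios to speed-to-RPM curves.
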(event VehicleAudio|EventOnInit
  (|CallParentFunction)
  (CallFunction|SetThrottleStateMotorSim)
  (CallFunction|SetResistanceMotorSim)
  (CallFunction|SetReverseMotorSim)
  (CallFunction|SetRevLimiterMotorSim)
  (CallFunction|SetMotorPhysicsSim)
  (CallFunction|SetVelocitySyncMotorSim))

(event Custom|SetThrottleStateMotorSim
  (bind _returnvalue (VehicleSound|GetAudioController))
  (bind _returnvalue_2 (Actor|GetComponentbyClass _returnvalue "/Script/AudioMotorSimStandardComponents.ThrottleStateMotorSimComponent"))
  ;[1 1 1] Throttle State
  (Utilities|IsValid _returnvalue_2
    (:Is%20Valid
      (Variables|/Script/AudioMotorSimStandardComponents.ThrottleStateMotorSimComponent|SetBlowoffMinThrottleTime (Variables|ThrottleState|GetBlowoff%20Min%20Throttle%20Time) _returnvalue_2))))

(event Custom|SetResistanceMotorSim
  (bind _returnvalue (VehicleSound|GetAudioController))
  (bind _returnvalue_2 (Actor|GetComponentbyClass _returnvalue "/Script/AudioMotorSimStandardComponents.ResistanceMotorSimComponent"))
  ;[1 1 1] Resistance
  (Utilities|IsValid _returnvalue_2
    (:Is%20Valid
      (Variables|/Script/AudioMotorSimStandardComponents.ResistanceMotorSimComponent|SetUpSpeedMaxFriction (Variables|Resistance|GetUp%20Speed%20Max%20Friction) _returnvalue_2)
      (Variables|/Script/AudioMotorSimStandardComponents.ResistanceMotorSimComponent|SetMinSpeed (Variables|Resistance|GetMin%20Speed) _returnvalue_2)
      (Variables|/Script/AudioMotorSimStandardComponents.ResistanceMotorSimComponent|SetSideSpeedFrictionCurve (Variables|Resistance|GetSide%20Speed%20Friction%20Curve) _returnvalue_2))))

(event Custom|SetReverseMotorSim
  (bind _returnvalue (VehicleSound|GetAudioController))
  (bind _returnvalue_2 (Actor|GetComponentbyClass _returnvalue "/Script/AudioMotorSimStandardComponents.ReverseMotorSimComponent"))
  ;[1 1 1] Reverse
  (Utilities|IsValid _returnvalue_2
    (:Is%20Valid
      (Variables|/Script/AudioMotorSimStandardComponents.ReverseMotorSimComponent|SetReverseEngineResistanceModifier (Variables|Reverse|GetReverse%20Engine%20Resistance%20Modifier) _returnvalue_2))))

(event Custom|SetRevLimiterMotorSim
  (bind _returnvalue (VehicleSound|GetAudioController))
  (bind _returnvalue_2 (Actor|GetComponentbyClass _returnvalue "/Script/AudioMotorSimStandardComponents.RevLimiterMotorSimComponent"))
  ;[1 1 1] Rev Limiter
  (Utilities|IsValid _returnvalue_2
    (:Is%20Valid
      (Variables|/Script/AudioMotorSimStandardComponents.RevLimiterMotorSimComponent|SetLimitTime (Variables|RevLimiter|GetLimit%20Time) _returnvalue_2)
      (Variables|/Script/AudioMotorSimStandardComponents.RevLimiterMotorSimComponent|SetAirMaxThrottleTime (Variables|RevLimiter|GetAir%20Max%20Throttle%20Time) _returnvalue_2)
      (Variables|/Script/AudioMotorSimStandardComponents.RevLimiterMotorSimComponent|SetSideSpeedThreshold (Variables|RevLimiter|GetSide%20Speed%20Threshold) _returnvalue_2)
      (Variables|/Script/AudioMotorSimStandardComponents.RevLimiterMotorSimComponent|SetLimiterMaxRpm (Variables|RevLimiter|GetLimiter%20Max%20Rpm) _returnvalue_2)
      (Variables|/Script/AudioMotorSimStandardComponents.RevLimiterMotorSimComponent|SetbRevLimitOnReverse (Variables|RevLimiter|GetRevLimitOnReverse) _returnvalue_2)
      (Variables|/Script/AudioMotorSimStandardComponents.RevLimiterMotorSimComponent|SetReverseMaxRPM (Variables|RevLimiter|GetReverseMaxRPM) _returnvalue_2)
      (Variables|/Script/AudioMotorSimStandardComponents.RevLimiterMotorSimComponent|SetReverseRecoverRPM (Variables|RevLimiter|GetReverseRecoverRPM) _returnvalue_2)
      (Variables|/Script/AudioMotorSimStandardComponents.RevLimiterMotorSimComponent|SetbHoldRPMInReverse (Variables|RevLimiter|GetHoldRPMInReverse) _returnvalue_2))))

(event Custom|SetMotorPhysicsSim
  (bind _returnvalue (VehicleSound|GetAudioController))
  (bind _returnvalue_2 (Actor|GetComponentbyClass _returnvalue "/Script/AudioMotorSimStandardComponents.MotorPhysicsSimComponent"))
  ;[1 1 1] MotorPhysics
  (Utilities|IsValid _returnvalue_2
    (:Is%20Valid
      ;;[1 1 1] Force
      (Variables|/Script/AudioMotorSimStandardComponents.MotorPhysicsSimComponent|SetWeight (Variables|MotorPhysics|Force|GetWeight) _returnvalue_2)
      (Variables|/Script/AudioMotorSimStandardComponents.MotorPhysicsSimComponent|SetEngineTorque (Variables|MotorPhysics|Force|GetEngine%20Torque) _returnvalue_2)
      (Variables|/Script/AudioMotorSimStandardComponents.MotorPhysicsSimComponent|SetBrakingHorsePower (Variables|MotorPhysics|Force|GetBraking%20Horse%20Power) _returnvalue_2)
      ;;[1 1 1] Gears
      (Variables|/Script/AudioMotorSimStandardComponents.MotorPhysicsSimComponent|SetGearRatios (Variables|MotorPhysics|Gears|GetGear%20Ratios) _returnvalue_2)
      (Variables|/Script/AudioMotorSimStandardComponents.MotorPhysicsSimComponent|SetClutchedGearRatio (Variables|MotorPhysics|Gears|GetClutched%20Gear%20Ratio) _returnvalue_2)
      (Variables|/Script/AudioMotorSimStandardComponents.MotorPhysicsSimComponent|SetbUseInfiniteGears (Variables|MotorPhysics|Gears|GetUse%20Infinite%20Gears) _returnvalue_2)
      (Variables|/Script/AudioMotorSimStandardComponents.MotorPhysicsSimComponent|SetbAlwaysDownshiftToZerothGear (Variables|MotorPhysics|Gears|GetAlways%20Downshift%20To%20Zeroth%20Gear) _returnvalue_2)
      (Variables|/Script/AudioMotorSimStandardComponents.MotorPhysicsSimComponent|SetInfiniteGearRatio (Variables|MotorPhysics|Gears|GetInfinite%20Gear%20Ratio) _returnvalue_2)
      (Variables|/Script/AudioMotorSimStandardComponents.MotorPhysicsSimComponent|SetUpShiftMaxRpm (Variables|MotorPhysics|Gears|GetUp%20Shift%20Max%20Rpm) _returnvalue_2)
      (Variables|/Script/AudioMotorSimStandardComponents.MotorPhysicsSimComponent|SetDownShiftStartRpm (Variables|MotorPhysics|Gears|GetDown%20Shift%20Start%20Rpm) _returnvalue_2)
      ;;[1 1 1] Resistance
      (Variables|/Script/AudioMotorSimStandardComponents.MotorPhysicsSimComponent|SetClutchedForceModifier (Variables|MotorPhysics|Resistance|GetClutched%20Force%20Modifier) _returnvalue_2)
      (Variables|/Script/AudioMotorSimStandardComponents.MotorPhysicsSimComponent|SetClutchedDecelScale (Variables|MotorPhysics|Resistance|GetClutched%20Decel%20Scale) _returnvalue_2)
      (Variables|/Script/AudioMotorSimStandardComponents.MotorPhysicsSimComponent|SetEngineGearRatio (Variables|MotorPhysics|Resistance|GetEngine%20Gear%20Ratio) _returnvalue_2)
      (Variables|/Script/AudioMotorSimStandardComponents.MotorPhysicsSimComponent|SetEngineFriction (Variables|MotorPhysics|Resistance|GetEngine%20Friction) _returnvalue_2)
      (Variables|/Script/AudioMotorSimStandardComponents.MotorPhysicsSimComponent|SetGroundFriction (Variables|MotorPhysics|Resistance|GetGround%20Friction) _returnvalue_2)
      (Variables|/Script/AudioMotorSimStandardComponents.MotorPhysicsSimComponent|SetWindResistancePerVelocity (Variables|MotorPhysics|Resistance|GetWind%20Resistance%20Per%20Velocity) _returnvalue_2)
      ;;[1 1 1] Physics Sim
      (Variables|/Script/AudioMotorSimStandardComponents.MotorPhysicsSimComponent|SetThrottleInterpolationTime (Variables|MotorPhysics|Physics|GetThrottle%20Interpolation%20Time) _returnvalue_2)
      (Variables|/Script/AudioMotorSimStandardComponents.MotorPhysicsSimComponent|SetRpmInterpSpeed (Variables|MotorPhysics|Physics|GetRpm%20Interp%20Speed) _returnvalue_2))))

(event Custom|SetVelocitySyncMotorSim
  (bind _returnvalue (VehicleSound|GetAudioController))
  (bind _returnvalue_2 (Actor|GetComponentbyClass _returnvalue "/Script/AudioMotorSimStandardComponents.VelocitySyncMotorSimComponent"))
  ;[1 1 1] Velocity Sync
  (Utilities|IsValid _returnvalue_2
    (:Is%20Valid
      (Variables|/Script/AudioMotorSimStandardComponents.VelocitySyncMotorSimComponent|SetNoThrottleTime (Variables|VelocitySync|GetNo%20Throttle%20Time) _returnvalue_2)
      (Variables|/Script/AudioMotorSimStandardComponents.VelocitySyncMotorSimComponent|SetSpeedThreshold (Variables|VelocitySync|GetSpeed%20Threshold) _returnvalue_2)
      (Variables|/Script/AudioMotorSimStandardComponents.VelocitySyncMotorSimComponent|SetSpeedToRpmCurve (Variables|VelocitySync|GetSpeed%20To%20Rpm%20Curve) _returnvalue_2)
      (Variables|/Script/AudioMotorSimStandardComponents.VelocitySyncMotorSimComponent|SetInterpSpeed (Variables|VelocitySync|GetInterp%20Speed) _returnvalue_2)
      (Variables|/Script/AudioMotorSimStandardComponents.VelocitySyncMotorSimComponent|SetInterpTime (Variables|VelocitySync|GetInterp%20Time) _returnvalue_2)
      (Variables|/Script/AudioMotorSimStandardComponents.VelocitySyncMotorSimComponent|SetFirstGearThrottleThreshold (Variables|VelocitySync|GetFirst%20Gear%20Throttle%20Threshold) _returnvalue_2))))
