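; Vehicle sound component that drives engine on/off audio and audio parameters from seat occupancy and fuel state.
; It binds seat and fuel delegates on init, unbinding them on shutdown.
; Driver changes fade out previous one-shots and spawn start/stop sounds based on fuel and disabled state; fuel depletion or refueling controls the MetaSound engine loop.
; It also pushes driver, fuel and local-passenger parameters, broadcasting when local passenger presence changes.
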(event VehicleAudio|EventOnInit
  (bind _returnvalue (VehicleSound|GetVehicleActor))
  (|CallParentFunction)
  (Utilities|IsValid _returnvalue
    (:Is%20Valid
      (Vehicle|BindEventtoOnPawnEnteredVehicleSeat _returnvalue (EventDispatchers|CreateEvent :!event On%20Pawn%20Entered%20Seat))
      (Vehicle|AssignOnPawnExitedVehicleSeat _returnvalue (EventDispatchers|CreateEvent :!event On%20Pawn%20Exited%20Seat))
      (EventDispatchers|BindEventtoOnOutOfFuelDelegate _returnvalue (EventDispatchers|CreateEvent :!event Out%20Of%20Fuel))
      (EventDispatchers|BindEventtoOnRefueledfromEmptyDelegate _returnvalue (EventDispatchers|CreateEvent :!event On%20Refueled%20From%20Empty))
      (EventDispatchers|BindEventtoOnLowFuelDelegate _returnvalue (EventDispatchers|CreateEvent :!event On%20Low%20Fuel)))))

(event VehicleAudio|EventOnShutdown
  (bind _returnvalue (VehicleSound|GetVehicleActor))
  (|CallParentFunction)
  (Utilities|IsValid _returnvalue
    (:Is%20Valid
      (Vehicle|UnbindEventfromOnPawnEnteredVehicleSeat _returnvalue (EventDispatchers|CreateEvent :!event On%20Pawn%20Entered%20Seat))
      (Vehicle|UnbindEventfromOnPawnExitedVehicleSeat _returnvalue (EventDispatchers|CreateEvent :!event On%20Pawn%20Exited%20Seat))
      (EventDispatchers|UnbindEventfromOnOutOfFuelDelegate _returnvalue (EventDispatchers|CreateEvent :!event Out%20Of%20Fuel))
      (EventDispatchers|UnbindEventfromOnRefueledfromEmptyDelegate _returnvalue (EventDispatchers|CreateEvent :!event On%20Refueled%20From%20Empty))
      (EventDispatchers|UnbindEventfromOnLowFuelDelegate _returnvalue (EventDispatchers|CreateEvent :!event On%20Low%20Fuel)))))

(fn AreAnyPassengersLocal
  (params (Target :type Fort%20Vehicle%20Interface%20Interface))
  (for _array_element (Vehicle|GetAllPassengers Target)
    (if (Pawn|IsLocallyControlled _array_element)
      (return true)))
  (return false))

(fn DriverEnterSeat
  (params (SeatIndex :type Integer) (PlayerPawn :type Fort%20Pawn%20Object%20Reference))
  (if (not (Vehicle|Fuel|IsVehicleFuelEmpty (VehicleSound|GetVehicleActor)))
    (if (== SeatIndex 0)
      (Variables|Default|SetIsDriver (Pawn|IsLocallyControlled PlayerPawn))
      (CallFunction|StopEngineAudio)
      (if (not (Vehicle|IsDisabled (VehicleSound|GetVehicleActor)))
        (bind _returnvalue (Audio|SpawnSoundAttached :Sound (Variables|Data|GetEngineOnSound) :AttachToComponent (Variables|/Script/Engine.Actor|GetRootComponent (Components|GetOwner)) :AttachPointName "None" :LocationType "KeepRelativeOffset" :bStopWhenAttachedToDestroyed true :VolumeMultiplier 1.0 :PitchMultiplier 1.0 :StartTime 0.0 :bAutoDestroy true))
        (Variables|Default|SetEngineOnAudio _returnvalue))
      (Driver|UpdateDriverStatus))
    (else
      (CallFunction|StopAudioOutput))))

(fn DriverExitSeat
  (params (SeatIndex :type Integer))
  (bind _returnvalue (VehicleSound|GetVehicleActor))
  (if (== SeatIndex 0)
    (Variables|Default|SetIsDriver false)
    (CallFunction|StopEngineAudio)
    (if (Vehicle|IsDisabled _returnvalue)
      _
      (elif (not (Vehicle|Fuel|IsVehicleFuelEmpty _returnvalue))
        (CallFunction|PlayEngineOffSound)))
    (Driver|UpdateDriverStatus)))

(fn GetAreAnyPassengersLocal
  (return (Variables|Default|GetPassengerIsLocal)))

(fn OnFuelChange
  (if (not (Vehicle|Fuel|IsVehicleFuelEmpty (VehicleSound|GetVehicleActor)))
    (if (Vehicle|HasDriver (VehicleSound|GetVehicleActor))
      (if (VehicleSound|GetUseMetasounds)
        (if (not (Audio|Components|Audio|IsPlaying (VehicleSound|GetEngineAudioComponent)))
          (CallFunction|StartEngineLoop))))
    (else
      (CallFunction|StopAudioOutput)
      (CallFunction|StopEngineLoop))))

(fn OnLowFuel
  (Fuel|UpdateFuel))

(fn OnPawnEnteredSeat
  (params (Vehicle :type Fort%20Vehicle%20Interface%20Interface%20%28by%20ref%29) (PlayerPawn :type Fort%20Pawn%20Object%20Reference) (SeatIndex :type Integer))
  ;[1 1 1] When player enters seat
  (Fuel|UpdateFuel)
  (Driver|DriverEnterSeat SeatIndex PlayerPawn)
  (Seat|UpdateAreAnyPassengersLocal Vehicle))

(fn OnPawnExitedSeat
  (params (Vehicle :type Fort%20Vehicle%20Interface%20Interface%20%28by%20ref%29) (PlayerPawn :type Fort%20Pawn%20Object%20Reference) (SeatIndex :type Integer))
  ;[1 1 1] When Player Leaves Seat
  (Fuel|UpdateFuel)
  (Driver|DriverExitSeat SeatIndex)
  (Seat|UpdateAreAnyPassengersLocal Vehicle))

(fn OnRefueledFromEmpty
  (Fuel|UpdateFuel)
  (Fuel|OnFuelChange))

(fn OutOfFuel
  (Fuel|UpdateFuel)
  (Fuel|OnFuelChange)
  (CallFunction|PlayEngineOffSound))

(fn PlayEngineOffSound
  (bind _returnvalue (Audio|SpawnSoundAttached :Sound (Variables|Data|GetEngineOffSound) :AttachToComponent (Variables|/Script/Engine.Actor|GetRootComponent (Components|GetOwner)) :AttachPointName "None" :LocationType "KeepRelativeOffset" :bStopWhenAttachedToDestroyed true :VolumeMultiplier 1.0 :PitchMultiplier 1.0 :StartTime 0.0 :bAutoDestroy true))
  (Variables|Default|SetEngineOffAudio _returnvalue))

(fn StartEngineLoop
  (if (VehicleSound|GetUseMetasounds)
    (Audio|Components|Audio|FadeIn (VehicleSound|GetEngineAudioComponent) 0.25 1.0 0.0 "Linear" :!class /Script/Engine.AudioComponent)))

(fn StopAudioOutput
  (CallFunction|StopEngineAudio))

(fn StopEngineAudio
  (bind _engineonaudio (Variables|Default|GetEngineOnAudio)
    (:then
      (Audio|Components|Audio|FadeOut _engineonaudio (Variables|Data|GetEnigneOnFadeOut) 0.0 "Linear" :!class /Script/Engine.AudioComponent))
    (:else
      _))
  (bind _engineoffaudio (Variables|Default|GetEngineOffAudio)
    (:then
      (Audio|Components|Audio|FadeOut _engineoffaudio (Variables|Data|GetEnigneOffFadeOut) 0.0 "Linear" :!class /Script/Engine.AudioComponent))
    (:else)))

(fn StopEngineLoop
  (if (VehicleSound|GetUseMetasounds)
    (Audio|Components|Audio|FadeOut (VehicleSound|GetEngineAudioComponent) 0.0 0.0 "Linear" :!class /Script/Engine.AudioComponent)))

(fn UpdateDriverStatus
  (Audio|Parameter|SetFloatParameter (VehicleSound|GetAudioParameterComponent) "HasDriver" (Math|Conversions|ToFloat(Boolean) (Vehicle|HasDriver (VehicleSound|GetVehicleActor))) :!class /Script/AudioExtensions.AudioParameterControllerInterface))

(fn UpdateFuel
  (bind _returnvalue (VehicleSound|GetAudioParameterComponent))
  (Utilities|IsValid _returnvalue
    (:Is%20Valid
      (Audio|Parameter|SetFloatParameter _returnvalue (Variables|Data|GetFuelName) (Math|Conversions|ToFloat(Boolean) (> (Vehicle|Fuel|GetVehicleFuelPercent (VehicleSound|GetVehicleActor)) 0)) :!class /Script/AudioExtensions.AudioParameterControllerInterface))))

(fn UpdateAreAnyPassengersLocal
  (params (Target :type Fort%20Vehicle%20Interface%20Interface))
  (bind _result (Seat|AreAnyPassengersLocal Target))
  (if (!= _result (Variables|Default|GetPassengerIsLocal))
    (bind _output_get (Variables|Default|SetPassengerIsLocal _result))
    (Audio|Parameter|SetBooleanParameter (VehicleSound|GetAudioParameterComponent) "bIsPassengerLocal" _output_get :!class /Script/AudioExtensions.AudioParameterControllerInterface)
    (Default|CallOnLocalPassengersChanged _output_get)))
