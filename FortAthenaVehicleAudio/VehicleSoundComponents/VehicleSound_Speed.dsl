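; Vehicle sound component that adds gear shift, rev limiter and reversing sounds, attached to the audio controller's engine audio component.
; On init it binds to the motor sim's gear change and rev limiter events, unbinding them on shutdown.
; Each shift or limiter hit fades out its previous sound before spawning another, with upshift audio as the shift fallback.
; Reverse audio starts after a delay if none exists, and fades out when reversing ends.
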(event VehicleAudio|EventOnInit
  (bind _returnvalue (VehicleSound|GetAudioController))
  (|CallParentFunction)
  (PhysicsSim|AssignOnGearChangedEvent (Actor|GetComponentbyClass _returnvalue "/Script/AudioMotorSimStandardComponents.MotorPhysicsSimComponent") (EventDispatchers|CreateEvent :!event OnGearShiftChange))
  (EventDispatchers|BindEventtoOnRevLimiterHit (Actor|GetComponentbyClass _returnvalue "/Script/AudioMotorSimStandardComponents.RevLimiterMotorSimComponent") (EventDispatchers|CreateEvent :!event OnRevLimiterHit)))

(event VehicleAudio|EventOnShutdown
  (bind _returnvalue (VehicleSound|GetAudioController))
  (PhysicsSim|UnbindEventfromOnGearChangedEvent (Actor|GetComponentbyClass _returnvalue "/Script/AudioMotorSimStandardComponents.MotorPhysicsSimComponent") (EventDispatchers|CreateEvent :!event OnGearShiftChange))
  (EventDispatchers|UnbindEventfromOnRevLimiterHit (Actor|GetComponentbyClass _returnvalue "/Script/AudioMotorSimStandardComponents.RevLimiterMotorSimComponent") (EventDispatchers|CreateEvent :!event OnRevLimiterHit)))

(event EventOnReversedChanged
  (params (bReversing :type Boolean))
  (if bReversing
    (Utilities|FlowControl|Delay (Variables|Reverse|GetReversingTriggerTime))
    (if bReversing
      (Variables|Default|GetReverseAudio
        (:then)
        (:else
          (bind _reversesound (Variables|Reverse|GetReverseSound)
            (:then
              (bind _returnvalue (Audio|SpawnSoundAttached _reversesound (AudioController|GetEngineAudioComponent (VehicleSound|GetAudioController))))
              (Variables|Default|SetReverseAudio _returnvalue))
            (:else)))))
    (else
      (bind _reverseaudio (Variables|Default|GetReverseAudio)
        (:then
          (Audio|Components|Audio|FadeOut _reverseaudio (Variables|Reverse|GetReverseFadeOutDuration) 0.0 "Linear" :!class /Script/Engine.AudioComponent))
        (:else)))))

(fn OnGearShiftChange
  (params (NewGear :type Integer))
  (bind _currentgear (Variables|Private|GetCurrentGear))
  (bind _returnvalue (> NewGear _currentgear))
  (bind _gearshiftupsound (Variables|Gear|GetGearShiftUpSound))
  (bind _gearshiftdownsound (Variables|Gear|GetGearShiftDownSound))
  (bind _returnvalue_2 (select _returnvalue _gearshiftupsound _gearshiftdownsound))
  (Utilities|IsValid _returnvalue_2
    (:Is%20Valid
      _)
    (:Is%20Not%20Valid
      (bind _gearshiftupsound_2 (Variables|Gear|GetGearShiftUpSound)
        (:then
          _)
        (:else))))
  (bind _shiftaudio (Variables|Default|GetShiftAudio)
    (:then
      (Audio|Components|Audio|FadeOut _shiftaudio 0.25 0.0 "Linear" :!class /Script/Engine.AudioComponent))
    (:else
      _))
  (bind _returnvalue_3 (Audio|SpawnSoundAttached (select (Utilities|IsValid _returnvalue_2 :!class /Script/Engine.KismetSystemLibrary) _returnvalue_2 _gearshiftupsound_2) (AudioController|GetEngineAudioComponent (VehicleSound|GetAudioController))))
  (Variables|Default|SetShiftAudio _returnvalue_3)
  (Variables|Private|SetCurrentGear (Variables|GetNewGear)))

(fn OnRevLimiterHit
  (bind _revlimiter (Variables|Default|GetRevLimiter)
    (:then
      (Audio|Components|Audio|FadeOut _revlimiter 0.5 0.0 "Linear" :!class /Script/Engine.AudioComponent))
    (:else
      _))
  (bind _returnvalue (Audio|SpawnSoundAttached (Variables|Limiter|GetRevLimiterHitSound) (AudioController|GetEngineAudioComponent (VehicleSound|GetAudioController))))
  (Variables|Default|SetRevLimiter _returnvalue))
