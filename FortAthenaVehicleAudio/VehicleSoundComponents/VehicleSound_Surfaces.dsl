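; Vehicle sound component that plays tire and surface audio for Fortnite vehicles, with handling for water, springs, popped tires and stopping.
; On init it binds mod and drift events and copies tuning into a native context struct; several comments mark logic as nativized.
; Data assets, default or per tire mod, choose the sound from surface type, water and air state.
; Two alternating audio components crossfade between sounds, driven by speed and tire parameters.
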(event VehicleAudio|EventOnInit
  (bind _returnvalue (VehicleSound|GetVehicleActor))
  (|CallParentFunction)
  (Utilities|IsValid _returnvalue
    (:Is%20Valid
      (Vehicle|Mods|BindEventtoOnModAppliedonClientDelegate _returnvalue (EventDispatchers|CreateEvent :!event OnModApplied))
      (Vehicle|Drift|AssignOnDriftStartedDelegate _returnvalue (EventDispatchers|CreateEvent :!event Custom|DriftStarted))
      (Vehicle|Drift|AssignOnDriftFinishedDelegate _returnvalue (EventDispatchers|CreateEvent :!event Custom|DriftStop))
      (CallFunction|SetNativeData)
      (Variables|Surfaces|Private|SetTires (Vehicle|Tire|GetNumTires (VehicleSound|GetVehicleActor))))))

(event VehicleAudio|EventOnShutdown
  (bind _returnvalue (VehicleSound|GetVehicleActor))
  (|CallParentFunction)
  (Utilities|IsValid _returnvalue
    (:Is%20Valid
      (Vehicle|Mods|UnbindEventfromOnModAppliedonClientDelegate _returnvalue (EventDispatchers|CreateEvent :!event OnModApplied))
      (Vehicle|Drift|UnbindEventfromOnDriftStartedDelegate _returnvalue (EventDispatchers|CreateEvent :!event Custom|DriftStarted))
      (Vehicle|Drift|UnbindEventfromOnDriftFinishedDelegate _returnvalue (EventDispatchers|CreateEvent :!event Custom|DriftStop)))))

(event VehicleAudio|EventOnUpdate
  (params (DeltaSeconds :type Float%20%28single-precision%29))
  (|CallParentFunction DeltaSeconds))

(event Custom|UpdateOnSurfaceChange
  (bind _tiremodtag (Variables|Surfaces|Private|GetTireModTag))
  (bind _tiremodassets (Variables|Surfaces|GetTireModAssets))
  (bind (_value _returnvalue) (Utilities|Map|Find _tiremodassets _tiremodtag))
  (bind _tiremod (Variables|Surfaces|Private|GetTireMod))
  (bind _returnvalue_2 (and _returnvalue _tiremod))
  (bind _surfacedefaultasset (Variables|Surfaces|GetSurfaceDefaultAsset))
  (bind _returnvalue_3 (select _returnvalue_2 _value _surfacedefaultasset))
  ;[1 1 1] [Nativized] New Surface setup - WIP - Uses surfaces to get data assets based on surfaces that should play.
  (if (not (Vehicle|IsAsleep (VehicleSound|GetVehicleActor)))
    (Utilities|IsValid _returnvalue_3
      (:Is%20Valid
        (bind _surfacesound (Class|BVehicleSurfaceDataAsset|GetSoundforSurface _returnvalue_3 (select (Variables|Water|Private|GetbIsInWater) "SurfaceType8" (Variables|/Script/FortniteGame.FortAthenaVehicle|GetSurfaceTypeVehicleOn (VehicleSound|GetVehicleActor))) (Variables|Air|GetInAir)))
        (Utilities|IsValid _surfacesound
          (:Is%20Valid
            (if (or (!= (Variables|Surfaces|Private|GetCurrentAudioSurface) _surfacesound) (not (Variables|Surfaces|Private|GetIsSurfaceSoundPlaying)))
              (bind _output_get (Variables|Surfaces|Private|SetCurrentAudioSurface _surfacesound))
              (Surface|PlaySurfaceAudio _output_get)))
          (:Is%20Not%20Valid
            (Development|PrintString "No Surface Sound Found!")
            (Surface|StopSurfaceAudio 0.25)
            (Variables|Surfaces|Private|SetCurrentAudioSurface 0)))))
    (else
      (Surface|StopSurfaceAudio 0.25))))

(event Custom|DriftStarted
  (params (bEnableVFX :type Boolean)))

(event Custom|DriftStop)

(fn UpdateTireCount
  (bind _returnvalue (VehicleSound|GetVehicleActor))
  (bind _returnvalue_2 (Vehicle|Tire|GetNumTires _returnvalue))
  (bind _returnvalue_3 (Math|Conversions|ToFloat(Integer) _returnvalue_2))
  (bind _tires (Variables|Surfaces|Private|GetTires))
  (bind _returnvalue_4 (Math|Conversions|ToFloat(Integer) _tires))
  (bind _returnvalue_5 (Math|Float|SafeDivide _returnvalue_4 _returnvalue_3))
  (VehicleSoundSurface|Native|SetFloatParameteronAudioComponents "Tires" _returnvalue_5)
  (if (<= _returnvalue_5 0.75)
    (VehicleSoundSurface|Native|SetTriggerParameteronAudioComponents "TiresPopped")
    (else
      (VehicleSoundSurface|Native|SetTriggerParameteronAudioComponents "TiresRepaired"))))

(fn GetSurfaceAudioComponent
  (bind _surfaceaudiocomponents (Variables|Surfaces|Private|GetSurfaceAudioComponents))
  (bind _nativesurfaceaudio1 (Variables|FortAthenaVehicleSoundComponent_Surface|GetNativeSurfaceAudio1))
  (bind _nativesurfaceaudio2 (Variables|FortAthenaVehicleSoundComponent_Surface|GetNativeSurfaceAudio2))
  ;[1 1 1] Added surface audio to component array, if any exists we can output which one to update and use.
  (Utilities|FlowControl|Sequence
    (:then_0
      (Utilities|IsValid _nativesurfaceaudio1
        (:Is%20Valid
          (Utilities|Array|AddUnique _surfaceaudiocomponents _nativesurfaceaudio1))))
    (:then_1
      (Utilities|IsValid _nativesurfaceaudio2
        (:Is%20Valid
          (Utilities|Array|AddUnique _surfaceaudiocomponents _nativesurfaceaudio2))))
    (:then_2
      (return (Variables|Surfaces|Private|GetSurfaceAudioComponents)))))

(fn HasLostTraction
  (return (and (not (Class|FortAthenaVehicle|HasTraction (Utilities|Casting|CastToFortAthenaSKMotorVehicle (VehicleSound|GetVehicleActor)))) (not (Vehicle|Tire|AreAnyTiresPopped (VehicleSound|GetVehicleActor))))))

(fn IsAccelerating
  (return (Math|Boolean|ANDBoolean (Math|Float|InRange(Float) (Vehicle|GetForwardSpeedKmh (VehicleSound|GetVehicleActor)) (Variables|Surfaces|Skid|GetMinSpeedForSkidFX) (Variables|Surfaces|Skid|GetMaxSpeedForSkidFX)) (Vehicle|GasPedalIsDown (VehicleSound|GetVehicleActor) :!class /Script/FortniteGame.FortVehicleInterface) (>= (Vehicle|Fuel|GetFuel (VehicleSound|GetVehicleActor)) 0.01) (not (Vehicle|VehicleIsinAir (VehicleSound|GetVehicleActor))))))

(fn IsBoostingOnGround
  (return (and (Class|FortVehicleInterface|IsBoosting (VehicleSound|GetVehicleActor)) (Class|FortVehicleInterface|IsTouchingGround (VehicleSound|GetVehicleActor)))))

(fn IsVehicleStopping
  (params (HardStopThreshold :type Float%20%28double-precision%29) (SoftStopThreshold :type Float%20%28double-precision%29))
  (bind _returnvalue (VehicleSound|GetVehicleActor))
  (Utilities|IsValid _returnvalue
    (:Is%20Valid
      (return (and (or (Vehicle|PlayerIsTryingtoReverse _returnvalue) (not (Vehicle|PlayerIsTryingtoMoveForward _returnvalue))) (<= (Vehicle|GetTotalSpeedKmh _returnvalue) (Math|Float|SelectFloat HardStopThreshold SoftStopThreshold (Vehicle|PlayerIsTryingtoReverse _returnvalue)))) (Vehicle|PlayerIsTryingtoReverse _returnvalue)))))

(fn OnEnterBodyWater
  (params (Vehicle :type Fort%20Athena%20Vehicle%20Object%20Reference) (WaterBody :type Fort%20Water%20Body%20Actor%20Object%20Reference))
  (Variables|Water|Private|SetbIsInWater true)
  (Audio|PlaySoundatLocation (Variables|Water|GetWaterEnterSound) (Transformation|GetActorLocation (Components|GetOwner))))

(fn OnExitBodyWater
  (params (Vehicle :type Fort%20Athena%20Vehicle%20Object%20Reference) (WaterBody :type Fort%20Water%20Body%20Actor%20Object%20Reference))
  (Variables|Water|Private|SetbIsInWater false)
  (Audio|PlaySoundatLocation (Variables|Water|GetWaterExitSound) (Transformation|GetActorLocation (Components|GetOwner)))
  (Surface|OnSurfaceTypeChange (Variables|/Script/FortniteGame.FortAthenaVehicle|GetSurfaceTypeVehicleOn (VehicleSound|GetVehicleActor))))

(fn OnModApplied
  (params (Vehicle :type Actor%20Object%20Reference) (ModTag :type Gameplay%20Tag%20Structure%20%28by%20ref%29))
  (bind (_value _returnvalue) (Utilities|Map|Find (Variables|Surfaces|GetTireModAssets) (Variables|Surfaces|Private|GetTireModTag)))
  ;[1 1 1] When mod is applied check for tire change to update surface audio.
  (if (GameplayTags|MatchesTag ModTag "(TagName=\"Vehicle.Mod.Tire\")")
    (Variables|Surfaces|Private|SetTireModTag ModTag)
    (bind _output_get (Variables|Surfaces|Private|SetTireMod (Vehicle|Mod|Tire|HasTireMod (VehicleSound|GetVehicleActor))))
    (Utilities|Struct|SetmembersinFortVehicleSoundComponentSurfaceContext (Variables|VehicleSound|Native|GetContext) _output_get (Variables|Surfaces|GetSurfaceDefaultAsset) (select _returnvalue _value (Variables|Surfaces|GetSurfaceDefaultAsset)))))

(fn OnSleepStateChange
  (params (bIsAwake :type Boolean))
  (if (not bIsAwake)
    (Surface|StopSurfaceAudio 0.25)))

(fn OnSpringCompression
  (bind _returnvalue (VehicleSoundSurface|Native|GetSpringCompressionMagnitude 1.2))
  (Variables|Surfaces|Private|SetSpringCompressionMagnitude _returnvalue)
  (Utilities|Time|SetTimerbyEvent (EventDispatchers|CreateEvent :!event ResetJitterMagnitude) (Variables|Surfaces|GetResetSpringCompressionMagnitudeTime)))

(fn OnSurfaceTypeChange
  (params (SurfaceTypeVehicleIsOn :type EPhysicalSurface%20Enum :default "SurfaceType_Default"))
  (Variables|Surfaces|Private|SetSurface%20Type%20Vehicle%20On SurfaceTypeVehicleIsOn))

(fn OnTirePopped
  (params (TireIndex :type Integer))
  (bind _result (Math|Integer|DecrementInt (Variables|Surfaces|Private|GetTires))
    (:%20%20
      (Variables|Surfaces|Private|SetTires _result)
      (Tires|UpdateTireCount))))

(fn OnTireRepaired
  (params (TireIndex :type Integer))
  (bind _result (Math|Integer|IncrementInt (Variables|Surfaces|Private|GetTires))
    (:%20%20
      (Variables|Surfaces|Private|SetTires _result)
      (Tires|UpdateTireCount))))

(fn PlaySurfaceAudio
  (params (SoundCue :type Sound%20Base%20Object%20Reference))
  ;[1 1 1] Swap between surface components to allow proper stop and starts (smooths out audio).
  (if (Variables|Surfaces|Private|GetSwapSurface)
    (bind _nativesurfaceaudio2 (Variables|FortAthenaVehicleSoundComponent_Surface|GetNativeSurfaceAudio2)
      (:then
        (Audio|Components|Audio|FadeOut _nativesurfaceaudio2 0.25 0.0 "Linear" :!class /Script/Engine.AudioComponent))
      (:else
        _))
    (bind _returnvalue (Audio|SpawnSoundAttached (Variables|GetSoundCue) (Variables|/Script/Engine.Actor|GetRootComponent (Components|GetOwner))))
    (bind _output_get (Variables|FortAthenaVehicleSoundComponent_Surface|SetNativeSurfaceAudio1 _returnvalue))
    (Audio|Components|Audio|FadeIn _output_get 0.15 1.0 0.0 "Linear" :!class /Script/Engine.AudioComponent)
    (Variables|Surfaces|Private|SetSwapSurface false)
    (Variables|Surfaces|Private|SetIsSurfaceSoundPlaying true)
    (else
      (bind _nativesurfaceaudio1 (Variables|FortAthenaVehicleSoundComponent_Surface|GetNativeSurfaceAudio1)
        (:then
          (Audio|Components|Audio|FadeOut _nativesurfaceaudio1 0.25 0.0 "Linear" :!class /Script/Engine.AudioComponent))
        (:else
          _))
      (bind _returnvalue_2 (Audio|SpawnSoundAttached (Variables|GetSoundCue) (Variables|/Script/Engine.Actor|GetRootComponent (Components|GetOwner))))
      (bind _output_get_2 (Variables|FortAthenaVehicleSoundComponent_Surface|SetNativeSurfaceAudio2 _returnvalue_2))
      (Audio|Components|Audio|FadeIn _output_get_2 0.15 1.0 0.0 "Linear" :!class /Script/Engine.AudioComponent)
      (Variables|Surfaces|Private|SetSwapSurface true)
      (Variables|Surfaces|Private|SetIsSurfaceSoundPlaying true))))

(fn ResetJitterMagnitude
  (Variables|Surfaces|Private|SetSpringCompressionMagnitude 0.0))

(fn SetNativeData
  (Utilities|Struct|SetmembersinFortVehicleSoundComponentSurfaceContext (Variables|VehicleSound|Native|GetContext) (Variables|Surfaces|Brake|GetHardStopThreshold) (Variables|Surfaces|Brake|GetSoftStopThreshold) (Variables|Surfaces|GetSurfaceParamInterpSpeed) (Variables|Surfaces|Private|GetSurfaceMap) "IsInWater" (Variables|Water|GetWaterParamInterpSpeed) "IsInAir" (Variables|Air|GetAirInterpSpeedAttack) (Variables|Air|GetAirInterpSpeedRelease) (Variables|Surfaces|Skid|GetMinSpeedForSkidFX) (Variables|Surfaces|Skid|GetMaxSpeedForSkidFX) "Skidding" "Skid" "StopSkid" (Variables|Surfaces|Skid|GetSkiddingParamInterpSpeed) "SurfaceMagnitude" (Variables|Surfaces|GetSpringCompressionMagnitudeAttack) (Variables|Surfaces|GetSpringCompressionMagnitudeRelease) 1.2 (Variables|Surfaces|GetResetSpringCompressionMagnitudeTime) 0.01 (Variables|Air|GetMinAirTimeForLandSound) "Surface" (Variables|Surfaces|Brake|GetBrakeStopFadeOutDuration) (Variables|Surfaces|Brake|GetHardStopThreshold) (Variables|Surfaces|Brake|GetSoftStopSound) (Variables|Surfaces|Brake|GetHardStopSound) (Variables|Surfaces|GetSurfaceDefaultAsset) (Utilities|Map|Find (Variables|Surfaces|GetTireModAssets) (Variables|Surfaces|Private|GetTireModTag)) "TiresPopped" "TiresRepaired" "Tires" "OnStop" "Speed" (Variables|Water|GetWaterEnterSound) (Variables|Water|GetWaterExitSound) "LandSpeed" "LandVelocity" (Variables|Surfaces|GetLandSound) (Variables|Surfaces|GetLandSoundLocationOffset)))

(fn SetSurfaceID
  (params (Surface%20ID :type Integer))
  ;[1 1 1] Should not be needed with new setup.
  (Variables|Surfaces|Private|SetSurface%20ID Surface%20ID))

(fn StopSurfaceAudio
  (params (FadeOutDuration :type Float%20%28single-precision%29))
  (bind _nativesurfaceaudio1 (Variables|FortAthenaVehicleSoundComponent_Surface|GetNativeSurfaceAudio1))
  (bind _nativesurfaceaudio2 (Variables|FortAthenaVehicleSoundComponent_Surface|GetNativeSurfaceAudio2))
  (Variables|Surfaces|Private|SetIsSurfaceSoundPlaying false)
  (VehicleSoundSurface|Native|SetTriggerParameteronAudioComponents "OnStop")
  (bind _output_get (Variables|Surfaces|Private|SetSurfaceSpeed 0.0))
  (VehicleSoundSurface|Native|SetFloatParameteronAudioComponents "Speed" _output_get)
  (Utilities|IsValid _nativesurfaceaudio1
    (:Is%20Valid
      (Audio|Components|Audio|FadeOut _nativesurfaceaudio1 FadeOutDuration 0.0 "Linear" :!class /Script/Engine.AudioComponent))
    (:Is%20Not%20Valid
      _))
  (Utilities|IsValid _nativesurfaceaudio2
    (:Is%20Valid
      (Audio|Components|Audio|FadeOut _nativesurfaceaudio2 FadeOutDuration 0.0 "Linear" :!class /Script/Engine.AudioComponent))))

(fn UpdateSurface
  (params (DeltaTime :type Float%20%28double-precision%29))
  (bind _returnvalue (VehicleSound|GetVehicleActor))
  (bind _returnvalue_2 (Class|FortVehicleInterface|IsinWater _returnvalue))
  (bind _returnvalue_3 (VehicleSound|GetVehicleActor))
  (bind _surfacetypevehicleon (Variables|/Script/FortniteGame.FortAthenaVehicle|GetSurfaceTypeVehicleOn _returnvalue_3))
  ;[1 1 1] [Nativized] Check speed of vehicle is we should play surface audio. Don't waste voice if vehicle is not in motion
  (Utilities|FlowControl|Sequence
    (:then_0
      (bind (_outnormalizedvehiclespeed _returnvalue_4) (VehicleSoundSurface|Native|CalculateSurfaceSpeed (Variables|Surfaces|Private|GetSurfaceSpeed) (Variables|Surfaces|GetSurfaceParamInterpSpeed) DeltaTime))
      (LocalVariables|SetInterpSurfaceSpeed _returnvalue_4)
      (if (> _outnormalizedvehiclespeed 0.01)
        (Tires|UpdateTireCount)
        (if (not (Variables|Surfaces|Private|GetIsSurfaceAudioPlaying))
          (Surface|OnSurfaceTypeChange (Variables|/Script/FortniteGame.FortAthenaVehicle|GetSurfaceTypeVehicleOn (VehicleSound|GetVehicleActor)))
          (Variables|Surfaces|Private|SetIsSurfaceAudioPlaying true))
        (else
          (Variables|Surfaces|Private|SetIsSurfaceAudioPlaying false)
          (Surface|StopSurfaceAudio 0.35))))
    (:then_1
      (if (!= (Variables|Surfaces|Private|GetSurfaceSpeed) (LocalVariables|GetInterpSurfaceSpeed))
        (bind _output_get (Variables|Surfaces|Private|SetSurfaceSpeed (LocalVariables|GetInterpSurfaceSpeed)))
        (VehicleSoundSurface|Native|SetFloatParameteronAudioComponents "Speed" _output_get)
        (if (Utilities|Enum|NotEqual(Enum) (Variables|Surfaces|Private|GetSurface%20Type%20Vehicle%20On) _surfacetypevehicleon)
          (Surface|OnSurfaceTypeChange _surfacetypevehicleon))))
    (:then_2
      ;;[1 1 1] Check for if we are water to update surface.
      (if (!= _returnvalue_2 (Variables|Water|Private|GetbIsInWater))
        (Variables|Water|Private|SetbIsInWater _returnvalue_2)
        (CallFunction|UpdateonSurfaceChange)))))

(fn UpdateSurfaceMagnitude
  (params (DeltaTime :type Float%20%28double-precision%29))
  (bind _returnvalue (VehicleSoundSurface|Native|UpdateSurfaceMagnitudeNative (Variables|Surfaces|Private|GetSurfaceMagnitudeParam) (Variables|Surfaces|Private|GetSpringCompressionMagnitude) (Variables|Surfaces|GetSpringCompressionMagnitudeAttack) (Variables|Surfaces|GetSpringCompressionMagnitudeRelease) DeltaTime))
  (Variables|Surfaces|Private|SetSurfaceMagnitudeParam _returnvalue))
